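# Translucent, unlit, two-sided material for the glowing blade effect of the Silent Tempo pickaxe's Peril variant, usable on Niagara mesh particles and spline meshes.
# UV-derived slope gradients sample curve atlases, combining with panning noise-distorted normals to carve the energy shape.
# Edge detail and inner electric noise blend top/bottom colour pairs for emissive, boosted near the spline top.
# Opacity combines fresnel, depth fade, tip boost, particle alpha and dynamic-parameter falloff.
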Material /BRCosmetics/Weapons/FORT_Melee/Pickaxe_SilentTempo/FX/M_Pickaxe_SilentTempo_Peril_Blade
graph {
 properties {
  bUsedWithMeshDeformer=true
  bUsedWithNiagaraMeshParticles=true
  bUsedWithSkeletalMesh=true
  bUsedWithSplineMeshes=true
  blendMode=BLEND_Translucent
  pixelDepthOffsetMode=PDOM_Legacy
  shadingModel=MSM_Unlit
  twoSided=true
  usedShadingModels=Unlit
 }
 nodes [
  n1 = TextureCoordinate()
  n2 = ScalarParameter() {defaultValue=0.7, group=UV_Control, parameterName="Slope(Normal)", sortPriority=2}
  n3 = ComponentMask(None=@n1) {r=true}
  n4 = Multiply(A=Abs(None=Subtract(A=@n3) {constB=0.5})) {constB=2}
  n5 = LinearInterpolate(
   A=Subtract(A=ComponentMask(None=@n1) {g=true}, B=Multiply(A=@n2, B=@n4))
   B=OneMinus(None=@n4)
   Alpha=ScalarParameter() {defaultValue=0.2, group=WPO, parameterName="Sprite(0)Spline(1)_WPO", sortPriority=1}
  )
  n6 = CurveAtlasRowParameter(CurveTime=@n5) {
   atlas=/Game/Weapons/FORT_Melee/Pickaxe_SilentTempo/FX/CurveAtlas_Pickaxe_SilentTempo
   curve=/Game/Weapons/FORT_Melee/Pickaxe_SilentTempo/FX/Curve_Pickaxe_SilentTempo_01
  }
  Add(A=@n6.G) {constB=0}
  n7 = ScalarParameter() {defaultValue=0.2, group=Distortion_Ctrl, parameterName=Distortion_Intensity}
  n8 = AppendVector(A=Multiply(A=OneMinus(None=@n4)) {constB=0.5}, B=@n6.G)
  n9 = ScalarParameter() {defaultValue=2, group=Distortion_Ctrl, parameterName=T_Distortion_U_Tile}
  n10 = ScalarParameter() {defaultValue=0.4, group=Distortion_Ctrl, parameterName=T_Distortion_V_Tile}
  n11 = Panner(
   Coordinate=Multiply(A=@n8, B=AppendVector(A=@n9, B=@n10))
   Speed=AppendVector(
    A=ScalarParameter() {defaultValue=-0.5, group=Distortion_Ctrl, parameterName=T_Distortion_U_Speed}
    B=ScalarParameter() {defaultValue=0.3, group=Distortion_Ctrl, parameterName=T_Distortion_V_Speed}
   )
  )
  n12 = AppendVector(A=VertexColor().R, B=Multiply(A=ScalarParameter() {defaultValue=0.05, parameterName=TimeSpeed}, B=Time()))
  n13 = TextureCoordinate()
  n14 = AppendVector(
   A=ScalarParameter() {defaultValue=0.9, group=Distortion_Ctrl, parameterName=T_Distortion_Sub_U_Tile}
   B=ScalarParameter() {defaultValue=0.6, group=Distortion_Ctrl, parameterName=T_Distortion_Sub_V_Tile}
  )
  n15 = ScalarParameter() {defaultValue=-0.5, group=Distortion_Ctrl, parameterName=T_Distortion_Sub_U_Speed}
  n16 = ScalarParameter() {defaultValue=0.3, group=Distortion_Ctrl, parameterName=T_Distortion_Sub_V_Speed}
  n17 = TextureSample(UVs=Panner(Coordinate=Multiply(A=@n13, B=@n14), Speed=AppendVector(A=@n15, B=@n16)) {speedY=0.3}) {texture=&T_NoisePack_01}
  n18 = ScalarParameter() {defaultValue=0.5, group=Distortion_Ctrl, parameterName=Distortion_Sub_Intensity}
  n19 = TextureSampleParameter2D(UVs=Add(A=Add(A=@n11, B=@n12), B=Multiply(A=@n17.R, B=@n18) {constB=0.1})) {
   group=Distortion_Ctrl
   parameterName=Dist Tex
   samplerType=SAMPLERTYPE_Normal
   texture=/Game/Global/Textures/Default/Tileable_Patterns/T_StormCloudDetail01_N
  }
  n20 = Add(
   A=Multiply(A=@n6.R, B=ScalarParameter() {defaultValue=1, group=Distortion_Ctrl, parameterName=BaseShape})
   B=Multiply(A=@n6.B, B=Multiply(A=@n19.R, B=@n7)) {constA=1}
  ) {constB=0}
  n21 = Multiply(A=@n20, B=ScalarParameter() {defaultValue=3, parameterName=Frequency}) {constB=2}
  n22 = CurveAtlasRowParameter(CurveTime=Saturate(None=@n21)) {
   atlas=/Game/Athena/LinearColorCurves/TempAthenaCurveAtlas
   curve=/Game/Gadgets/Assets/VinderTech_GliderChute/Glider_Donut_Plate/FX/Curve_Donut_Plate02
   defaultValue=6
   parameterName=DarkEdge
  }
  n23 = TextureCoordinate()
  n24 = AppendVector(
   A=ScalarParameter() {defaultValue=1, group=Edge_Detail, parameterName=T_Detail_U_Tile, sortPriority=2}
   B=ScalarParameter() {defaultValue=1, group=Edge_Detail, parameterName=T_Detail_V_Tile, sortPriority=2}
  )
  n25 = ScalarParameter() {defaultValue=0.2, group=Edge_Detail, parameterName=T_Detail_U_Speed, sortPriority=2}
  n26 = ScalarParameter() {defaultValue=0.8, group=Edge_Detail, parameterName=T_Detail_V_Speed, sortPriority=2}
  n27 = TextureSampleParameter2D(UVs=Panner(Coordinate=Multiply(A=@n23, B=@n24), Speed=AppendVector(A=@n25, B=@n26)) {speedY=0.1}) {
   group=Edge_Detail
   parameterName=Detail_map
   samplerType=SAMPLERTYPE_LinearColor
   texture=/Engine/Functions/Engine_MaterialFunctions02/ExampleContent/Textures/OffsetNoiseDistanceFields_ForNormals
  }
  n28 = OneMinus(None=Multiply(A=@n22.B, B=@n27.R))
  n29 = LinearInterpolate(
   A=Constant() {r=1}
   B=LinearInterpolate(A=ScalarParameter() {group=Edge_Detail, parameterName=Detail_Brightness}, Alpha=Saturate(None=@n28))
   Alpha=ScalarParameter() {defaultValue=1, group=Edge_Detail, parameterName=Edge_Detail_Switch}
  )
  n30 = ParticleColor()
  n31 = ScalarParameter() {defaultValue=1, parameterName=Final_Emissive_Intensity}
  n32 = TextureCoordinate()
  n33 = Subtract(
   A=OneMinus(None=ComponentMask(None=@n32) {g=true})
   B=ScalarParameter() {defaultValue=0.7, group=Spline_Top_White_Ctrl, parameterName=White_Offset}
  )
  n34 = Power(
   Base=Multiply(A=Saturate(None=@n33), B=ScalarParameter() {defaultValue=1, group=Spline_Top_White_Ctrl, parameterName=White_intensity})
   Exp=ScalarParameter() {defaultValue=1, group=Spline_Top_White_Ctrl, parameterName=White_Power}
  )
  n35 = LinearInterpolate(
   A=@n31
   B=Multiply(A=Add(A=@n34), B=@n31)
   Alpha=ScalarParameter() {defaultValue=0.2, group=WPO, parameterName="Sprite(0)Spline(1)_WPO", sortPriority=1}
  )
  n36 = CurveAtlasRowParameter(CurveTime=@n5) {
   atlas=&CurveAtlas_Pickaxe_SilentTempo
   curve=/Game/Weapons/FORT_Melee/Pickaxe_SilentTempo/FX/Curve_Pickaxe_SilentTempo_02
   defaultValue=1
   parameterName=Curve_DonutPlate_2
  }
  n37 = CurveAtlasRowParameter(CurveTime=Frac(None=@n21)) {
   atlas=&CurveAtlas_Pickaxe_SilentTempo
   curve=&Curve_Pickaxe_SilentTempo_01
  }
  n38 = VectorParameter() {parameterName=Light_Color_Top, defaultValue=<rgba 0 0.6667 2 1>}
  n39 = VectorParameter() {parameterName=Light_Color_Bottom, defaultValue=<rgba 0 1.5 3 1>}
  n40 = OneMinus(None=SphereMask(A=TextureCoordinate(), B=Constant2Vector() {r=0.5}) {attenuationRadius=0.65, hardnessPercent=0})
  n41 = CurveAtlasRowParameter(CurveTime=@n40) {
   atlas=&CurveAtlas_Pickaxe_SilentTempo
   curve=&Curve_Pickaxe_SilentTempo_02
   defaultValue=1
   parameterName=Curve_DonutPlate_2
  }
  n42 = Add(A=VertexColor().R, B=LinearInterpolate(A=TextureCoordinate(), B=@n8, Alpha=ScalarParameter() {parameterName=InnerEnergy_UVswitch}))
  n43 = Multiply(
   A=AppendVector(
    A=ScalarParameter() {defaultValue=1, group=Inner_Detail, parameterName=T_InnerEnergy_U_Tile}
    B=ScalarParameter() {defaultValue=1, group=Inner_Detail, parameterName=T_InnerEnergy_V_Tile}
   )
   B=@n42
  )
  n44 = AppendVector(
   A=ScalarParameter() {defaultValue=1, group=Inner_Detail, parameterName=T_InnerDetail_U_Tile}
   B=ScalarParameter() {defaultValue=1, group=Inner_Detail, parameterName=T_InnerDetail_V_Tile}
  )
  n45 = ScalarParameter() {group=Inner_Detail, parameterName=T_InnerDetail_U_Speed}
  n46 = ScalarParameter() {defaultValue=1, group=Inner_Detail, parameterName=T_InnerDetail_V_Speed}
  n47 = TextureSample(UVs=Panner(Coordinate=Multiply(A=@n42, B=@n44), Speed=AppendVector(A=@n45, B=@n46)) {speedY=1}) {samplerType=SAMPLERTYPE_LinearGrayscale, texture=/Game/Global/Textures/Default/Noise/T_SwirlNoise_N_height}
  n48 = TextureSampleParameter2D(UVs=Add(A=Panner(Coordinate=@n43) {speedY=0.5}, B=Multiply(A=@n47.R) {constB=0.2})) {
   group=Inner_Detail
   parameterName=T_Inner_Ennergy
   samplerType=SAMPLERTYPE_LinearColor
   texture=/Game/Characters/Player/Male/Medium/Bodies/M_MED_SRDrift/Textures/T_ElectricNoise
  }
  n49 = Clamp(None=Add(A=@n41.G, B=Multiply(A=@n40, B=@n48.R))) {maxDefault=10}
  n50 = VectorParameter() {parameterName=Dark_Top, defaultValue=<rgba 0 0.5 3 1>}
  n51 = VectorParameter() {parameterName=Dark_Bottom, defaultValue=<rgba 0 1 3 1>}
  n52 = Multiply(
   A=Multiply(A=@n36.R) {constB=100}
   B=LinearInterpolate(A=LinearInterpolate(A=@n38.RGB, B=@n39.RGB, Alpha=@n49), B=LinearInterpolate(A=@n50.RGB, B=@n51.RGB, Alpha=@n49), Alpha=@n37.A)
  )
  MaterialFunctionCall(Emissive=Multiply(A=Multiply(A=@n29, B=@n30.RGB), B=Multiply(A=@n35, B=@n52))) {materialFunction=&MF_FX_EmissiveAdjustments}
  n53 = OneMinus(None=ComponentMask(None=TextureCoordinate()) {g=true})
  n54 = Multiply(A=Saturate(None=Power(Base=@n53) {constExponent=10}), B=ScalarParameter() {defaultValue=10, parameterName=Tip Alpha})
  n55 = Abs(None=DotProduct(A=CameraVectorWS(), B=VertexNormalWS()))
  n56 = ScalarParameter() {defaultValue=0.3, group=Translucent, parameterName=Frenel}
  n57 = DynamicParameter() {defaultValue=<rgba 3 2 20 1>, paramNames [Power,Mult,DepthFade,Param4]}
  n58 = CurveAtlasRowParameter(CurveTime=@n21) {
   atlas=&CurveAtlas_Pickaxe_SilentTempo
   curve=&Curve_Pickaxe_SilentTempo_02
   defaultValue=1
   parameterName=Curve_DonutPlate_2
  }
  n59 = Power(Base=ComponentMask(None=TextureCoordinate()) {g=true}, Exp=ScalarParameter() {defaultValue=1, parameterName=Top_Softness})
  n60 = Constant() {r=1}
  n61 = ScalarParameter() {defaultValue=0.2, group=WPO, parameterName="Sprite(0)Spline(1)_WPO", sortPriority=1}
  n62 = LinearInterpolate(A=@n58.A, B=Multiply(A=@n21, B=LinearInterpolate(A=@n59, B=@n60, Alpha=@n61)), Alpha=ScalarParameter() {parameterName=Softness})
  n63 = ComponentMask(None=TextureCoordinate()) {g=true}
  n64 = Multiply(A=OneMinus(None=Power(Base=@n63, Exp=@n57.Power) {constExponent=3}), B=@n57.Mult) {constB=2}
  n65 = Multiply(
   A=LinearInterpolate(B=ScalarParameter() {defaultValue=1, group=Translucent, parameterName=Fresnel_Enhancement}, Alpha=Power(Base=@n55, Exp=@n56))
   B=Multiply(A=DepthFade(Opacity=@n62, FadeDistance=@n57.DepthFade) {fadeDistanceDefault=10}, B=Saturate(None=@n64))
  )
  Multiply(A=@n30.A, B=Add(A=@n65, B=Multiply(A=@n65, B=@n54)))
  Multiply()
  OneMinus(None=@n27.R)
  Reroute(None=@n62)
  Reroute(None=@n20)
  ScalarParameter() {defaultValue=20, group=Translucent, parameterName=Fade_Distance}
 ]
}
