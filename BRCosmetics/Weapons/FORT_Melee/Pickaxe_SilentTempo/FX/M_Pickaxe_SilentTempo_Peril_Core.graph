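# Translucent unlit material for Niagara mesh particles, apparently the glowing core effect of the SilentTempo Peril pickaxe.
# A Fresnel term blends two bright blue HDR colors, masked by an inverted, powered bounding-box height gradient with a small added floor.
# That result is tinted by particle color and passed through the shared FX emissive adjustments function.
# Particle alpha is multiplied by an Alpha parameter, apparently for opacity.
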Material /BRCosmetics/Weapons/FORT_Melee/Pickaxe_SilentTempo/FX/M_Pickaxe_SilentTempo_Peril_Core
graph {
 properties {
  bUsedWithNiagaraMeshParticles=true
  blendMode=BLEND_Translucent
  pixelDepthOffsetMode=PDOM_Legacy
  shadingModel=MSM_Unlit
  usedShadingModels=Unlit
 }
 nodes [
  n1 = OneMinus(None=MaterialFunctionCall() {materialFunction=&BoundingBoxBased_0-1_UVW}.B)
  n2 = Saturate(None=Add(A=Power(Base=@n1)) {constB=0.03})
  n3 = LinearInterpolate(
   A=VectorParameter() {parameterName=Color A, defaultValue=<rgba 2.4 6.1333 8 1>}.RGB
   B=VectorParameter() {parameterName=Color B, defaultValue=<rgba 0 0.6667 4 1>}.RGB
   Alpha=Fresnel() {baseReflectFraction=0, exponent=0.2}
  )
  n4 = ParticleColor()
  MaterialFunctionCall(Emissive=Multiply(A=Multiply(A=@n2, B=@n3), B=@n4.RGB)) {materialFunction=&MF_FX_EmissiveAdjustments}
  Multiply(A=@n4.A, B=ScalarParameter() {defaultValue=1, parameterName=Alpha})
 ]
}
